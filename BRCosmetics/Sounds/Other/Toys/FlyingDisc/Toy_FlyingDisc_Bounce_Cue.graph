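# Sound cue for the bounce of the flying disc toy, built from three disc landing samples.
# A Random node picks one of the equally weighted waves and feeds a Modulator that varies pitch and slightly varies reduced volume.
# The cue plays at a lowered overall volume with shared ball-bounce attenuation and is assigned to the emote music sound class.
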SoundCue /BRCosmetics/Sounds/Other/Toys/FlyingDisc/Toy_FlyingDisc_Bounce_Cue
root=Modulator_0
nodes [
 {id=Modulator_0, children [Random_0], properties {pitchMin=0.85, volumeMax=0.7, volumeMin=0.65}}
 {id=Random_0, children [WavePlayer_2,WavePlayer_3,WavePlayer_4], properties {weights [1,1,1]}}
 {id=WavePlayer_2, properties {soundWaveAssetPtr=&/Athena_Toy_FlyingDisc_Land_01}}
 {id=WavePlayer_3, properties {soundWaveAssetPtr=&/Athena_Toy_FlyingDisc_Land_02}}
 {id=WavePlayer_4, properties {soundWaveAssetPtr=&/Athena_Toy_FlyingDisc_Land_03}}
]
properties {
 attenuationSettings=/Game/Sounds/Attenuation/Fort_Ball_Bounce_Athena_Att
 maxDistance=7000
 soundClassObject=&Fort_Music_Emote
 volumeMultiplier=0.3
}
