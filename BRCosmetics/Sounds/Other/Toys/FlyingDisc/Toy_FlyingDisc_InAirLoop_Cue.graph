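# Sound cue for the looping in-air sound of the flying disc toy, routed to the Fort_Gadgets sound class.
# A looping wave feeds a Modulator that randomizes pitch downward at fixed volume, then an Enveloper that ramps volume up from half to full.
# The cue overrides attenuation with logarithmic distance falloff and caps concurrent instances per owner.
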SoundCue /BRCosmetics/Sounds/Other/Toys/FlyingDisc/Toy_FlyingDisc_InAirLoop_Cue
root=Enveloper_0
nodes [
 {
  id=Enveloper_0
  children [Modulator_0]
  properties {pitchCurve {editorCurveData {keys [{time=0}]}}, volumeCurve {editorCurveData {keys [<at 0 0.5>,<at 1 1>]}}}
 }
 {id=Modulator_0, children [WavePlayer_0], properties {pitchMax=0.5, pitchMin=0.45, volumeMax=0.5, volumeMin=0.5}}
 {id=WavePlayer_0, properties {bLooping=true, soundWaveAssetPtr=/Game/Athena/Sounds/Toys/FlyingDisc/Athena_Toy_FlyingDisc_InAirLoop_02}}
]
properties {
 attenuationSettings=/Game/Sounds/Attenuation/Fort_ToyHit_Att
 bOverrideAttenuation=true
 bOverrideConcurrency=true
 maxDistance=2000
 pitchMultiplier=1.2
 soundClassObject=&Fort_Gadgets
 volumeMultiplier=0.4
 attenuationOverrides {distanceAlgorithm=Logarithmic, falloffDistance=2000, stereoSpread=512, attenuationShapeExtents {x=0}}
 concurrencyOverrides {bLimitToOwner=true, maxCount=2, volumeScale=0.7}
}
