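# Sound cue for throwing the flying disc toy in the cosmetics toy set.
# A root Random node picks one of three throw wave variations with equal weight.
# Playback uses emote attenuation and the emote music sound class, with a slightly boosted volume and a limited audible distance.
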SoundCue /BRCosmetics/Sounds/Other/Toys/FlyingDisc/Toy_FlyingDisc_Throw_Cue
root=Random_0
nodes [
 {id=Random_0, children [WavePlayer_1,WavePlayer_3,WavePlayer_4], properties {weights [1,1,1]}}
 {id=WavePlayer_1, properties {soundWaveAssetPtr=&/Athena_Toy_FlyingDsc_Throw_01}}
 {id=WavePlayer_3, properties {soundWaveAssetPtr=&/Athena_Toy_FlyingDsc_Throw_02}}
 {id=WavePlayer_4, properties {soundWaveAssetPtr=&/Athena_Toy_FlyingDsc_Throw_03}}
]
properties {
 attenuationSettings=&Fort_Emotes_Athena_Att
 maxDistance=1600
 soundClassObject=&Fort_Music_Emote
 volumeMultiplier=1.25
}
